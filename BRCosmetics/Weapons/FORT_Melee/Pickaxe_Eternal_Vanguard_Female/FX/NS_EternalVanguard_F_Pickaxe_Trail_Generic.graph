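# Niagara trail for the female Eternal Vanguard pickaxe, drawing short-lived, fading cyan air ribbons from skeletal mesh bones or sockets.
# A single AirTrail emitter uses the shared AnimTrail modules to spawn, initialize and update particles, randomizing material parameters and narrowing width over lifetime.
# Two ribbon renderers accompany a secondary width and position attribute set, with position offset along ribbon facing and width shaped by emitter loop age and particle fade.
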NiagaraSystem /BRCosmetics/Weapons/FORT_Melee/Pickaxe_Eternal_Vanguard_Female/FX/NS_EternalVanguard_F_Pickaxe_Trail_Generic
dependencies {
 usedDataInterfaces [&NiagaraDataInterfaceCurve, &NiagaraDataInterfaceColorCurve]
 usedDynamicInputs []
 usedModules [&SystemState]
}
emitters {AirTrail {
  emitterUpdateScript [
   &EmitterState(Life Cycle Mode=::NewEnumerator1, Loop Behavior=::NewEnumerator1, Loop Duration=0.3, UseLoopDelay=false)
   &FN_SpawnAnimTrail_v2(Spawn Interval=0)
  ]
  particleSpawnScript [
   /Game/Effects/Niagara/Modules/AnimTrail/FN_InitializeAnimTrail/FN_InitializeAnimTrailBoneIndecies(First Bone or Socket Index=1)
   /Game/Effects/Niagara/Modules/AnimTrail/FN_InitializeAnimTrail/FN_InitializeAnimTrail(AnimTrailWidth=1, Lifetime=0.125)
   &DynamicMaterialParameters(
    Index 0 Param 1=&UniformRangedFloat(Minimum=0, Maximum=1)
    Index 0 Param 2=&UniformRangedFloat(Minimum=0, Maximum=1)
    Index 0 Param 3=1
    Index 0 Param 4=1
    Module.Param0WriteEnabled=true
    Module.Param1WriteEnabled=true
    Module.Param2WriteEnabled=true
    Module.Param3WriteEnabled=true
   )
   /SetVariables(Particles.SecondRibbonWidth:NiagaraFloat=&Multiply_Float(
     A=$Particles.RibbonWidth
     B=&FloatFromCurve(FloatCurve=DataInterface() {propertyValues {Curve {keys [<at 0 0>,<at 0.1 0.75>,<at 0.5 0>]}}}, CurveIndex=$Emitter.NormalizedLoopAge)
    )) {isSetParameters=true}
   /SetVariables(Particles.SecondRibbonPosition:NiagaraPosition=$Particles.Position) {isSetParameters=true}
  ]
  particleUpdateScript [
   &SolveForcesAndVelocity(Clamp Velocity=false, Limit Acceleration=false)
   &ScaleRibbonWidth(Ribbon Width Scale=&FloatFromCurve(FloatCurve=DataInterface() {propertyValues {Curve {keys [
         {arriveTangent=-1.2579, interpMode=RCIM_Cubic, time=0, value=1}
         {arriveTangent=-0.9664, interpMode=RCIM_Cubic, leaveTangent=-0.9664, tangentMode=RCTM_User, time=1, value=0.2}
        ]}}}))
   &FN_UpdateAnimTrail_v2()
   &Color(Color=&ColorFromCurve(ColorCurve=DataInterface() {propertyValues {
       AlphaCurve {keys [<at 0 0.15 RCIM_Cubic>,<at 1 0 RCIM_Cubic>]}
       BlueCurve {keys [<at 0.0114 1>]}
       GreenCurve {keys [<at 0.0114 0.854>]}
       RedCurve {keys [<at 0.0114 0.1611>]}
      }}))
   /SetVariables(Particles.SecondRibbonWidth:NiagaraFloat=&Multiply_Float(
     A=$Particles.Initial.SecondRibbonWidth
     B=&FloatFromCurve(FloatCurve=DataInterface() {propertyValues {Curve {keys [<at 0 1>,<at 1 0>]}}})
    )) {isSetParameters=true}
   /SetVariables(Particles.SecondRibbonPosition:NiagaraPosition=&Add_Vector(
     A=Linked() {linkedVariable {name=Particles.Position, type=NiagaraPosition}}
     B=&Multiply_VectorByFloat(Vector=$Particles.RibbonFacing, Float=30)
    )) {isSetParameters=true}
  ]
  rendererClasses [&NiagaraRibbonRendererProperties]
  renderers [
   {rendererClass=&NiagaraRibbonRendererProperties}
   {rendererClass=&NiagaraRibbonRendererProperties, rendererIndex=1}
  ]
 }}
summary {
 systemName=NS_EternalVanguard_F_Pickaxe_Trail_Generic
 emitters [{emitterName=AirTrail, rendererClasses [&NiagaraRibbonRendererProperties]}]
 userVariables [
  User.LifetimeCustom {type=NiagaraFloat, defaultValue=0, description=""}
  User.Primary Color {type=LinearColor, description="", defaultValue {struct=LinearColor, value=<rgba 1 1 1 1>}}
  User.RibbonIDOffset {type=NiagaraInt32}
  User.SkeletalMesh {
   type=NiagaraDataInterfaceSkeletalMesh
   defaultValue {
    struct=NiagaraExt_VariableValue_DataInterface
    value {dataInterface=&self:NiagaraDataInterfaceSkeletalMesh_0}
   }
  }
 ]
}
