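# Physics asset for the Eternal Vanguard Female pickaxe, giving its dynamic cable bones collision and joint limits for physical motion.
# Three small capsule bodies cover two cable bones and a shared pivot bone.
# Each cable is constrained directly to the pivot rather than chained to the other, with moderate swing and tight twist limits.
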PhysicsAsset /BRCosmetics/Weapons/FORT_Melee/Pickaxe_Eternal_Vanguard_Female/Meshes/Eternal_Vanguard_Female_Physics
bodies {
 Dyn_Cable_1 [{shapeType=Capsule, length=4, radius=1, center=<13.1052 5.3566 -2.2679>, rotation=<rot -13.9279 123.8522 -87.758>}]
 Dyn_Cable_2 [{shapeType=Capsule, length=2, radius=0.75, center=<4.5797 0.037 -0.4243>, rotation=<rot 41.1088 -89.9995 -81.1979>}]
 Dyn_Pivot_02 [{shapeType=Capsule, length=1, radius=1}]
}
constraints [
 {bone1Name=Dyn_Cable_1, bone2Name=Dyn_Pivot_02, swing1LimitDegrees=30, swing2LimitDegrees=30, twistLimitDegrees=5}
 {bone1Name=Dyn_Cable_2, bone2Name=Dyn_Pivot_02, swing1LimitDegrees=30, swing2LimitDegrees=30, twistLimitDegrees=5}
]
