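# Animation Blueprint for the Eternal Vanguard Female pickaxe that procedurally animates its satellite dish and pivot bones.
# A looping chain of timed delays toggles Stage variables, driving interpolated blends between fixed pivot-bone poses.
# The dish then gets an AnimSpinner rotation and a sine-driven master-bone roll, followed by RigidBody simulation.
# The DishTilt look-at and sway nodes are unconnected, and the interpolated Satelite_UD value is unused.
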AnimBlueprint /BRCosmetics/Weapons/FORT_Melee/Pickaxe_Eternal_Vanguard_Female/Meshes/Eternal_Vanguard_Female_AnimBP
target_skeleton=&/Eternal_Vanguard_Female_Skeleton
graphs [
 AnimGraph {nodes [
   Look At - Bone: Dyn_Satalite_DishTilt(Alpha=1) {
    alphaBoolBlend {blendOption=Linear}
    boneToModify {boneName=Dyn_Satalite_DishTilt}
    lookAtLocation {x=0}
    lookAtTarget {boneReference {boneName=Dyn_Dish_Lookat}}
    lookAt_Axis {axis=<1 0>}
   }
   "Transform (Modify) Bone - Bone: Dyn_Satalite_DishTilt"(Alpha=1, Translation=Get DishSway() -> (DishSway:Vector), Rotation="0,0,0", Scale="1,1,1") {
    rotationMode=BMM_Additive
    rotationSpace=BCS_BoneSpace
    translationMode=BMM_Additive
    translationSpace=BCS_BoneSpace
    alphaBoolBlend {blendOption=Linear}
    boneToModify {boneName=Dyn_Satalite_DishTilt}
   }
   "Transform (Modify) Bone - Bone: Dyn_Pivot_02"(Alpha=1, Translation="0,0,0", Rotation="0,-10,0", Scale="1,1,1") {
    rotationMode=BMM_Replace
    rotationSpace=BCS_BoneSpace
    scaleMode=BMM_Replace
    scaleSpace=BCS_BoneSpace
    translationMode=BMM_Replace
    translationSpace=BCS_BoneSpace
    alphaBoolBlend {blendOption=Linear}
    boneToModify {boneName=Dyn_Pivot_02}
    rotation {yaw=-10}
   }
   n1 = "Transform (Modify) Bone - Bone: Dyn_Pivot_02"(Alpha=1, Translation="0,0,0", Rotation="0,-15,-50", Scale="1,1,1") {
    rotationMode=BMM_Replace
    rotationSpace=BCS_BoneSpace
    scaleMode=BMM_Replace
    scaleSpace=BCS_BoneSpace
    translationMode=BMM_Replace
    translationSpace=BCS_BoneSpace
    alphaBoolBlend {blendOption=Linear}
    boneToModify {boneName=Dyn_Pivot_02}
    rotation {roll=-50, yaw=-15}
   }
   n2 = "Transform (Modify) Bone - Bone: Dyn_Pivot_01"(Alpha=1, Translation="0,0,0", Rotation="0,-10,0", Scale="1,1,1") {
    rotationMode=BMM_Replace
    rotationSpace=BCS_BoneSpace
    scaleMode=BMM_Replace
    scaleSpace=BCS_BoneSpace
    translationMode=BMM_Replace
    translationSpace=BCS_BoneSpace
    alphaBoolBlend {blendOption=Linear}
    boneToModify {boneName=Dyn_Pivot_01}
    rotation {yaw=-10}
   }
   n3 = "Transform (Modify) Bone - Bone: Dyn_Pivot_02"(Alpha=1, Translation="0,0,0", Rotation="0,-25,-15", Scale="1,1,1") {
    rotationMode=BMM_Replace
    rotationSpace=BCS_BoneSpace
    scaleMode=BMM_Replace
    scaleSpace=BCS_BoneSpace
    translationMode=BMM_Replace
    translationSpace=BCS_BoneSpace
    alphaBoolBlend {blendOption=Linear}
    boneToModify {boneName=Dyn_Pivot_02}
    rotation {roll=-15, yaw=-25}
   }
   n4 = "Transform (Modify) Bone - Bone: Dyn_Pivot_01"(Alpha=1, Translation="0,0,0", Rotation="0,15,0", Scale="1,1,1") {
    rotationMode=BMM_Replace
    rotationSpace=BCS_BoneSpace
    scaleMode=BMM_Replace
    scaleSpace=BCS_BoneSpace
    translationMode=BMM_Replace
    translationSpace=BCS_BoneSpace
    alphaBoolBlend {blendOption=Linear}
    boneToModify {boneName=Dyn_Pivot_01}
    rotation {yaw=15}
   }
   n5 = "Transform (Modify) Bone - Bone: Dyn_Pivot_02"(Alpha=1, Translation="0,0,0", Rotation="0,-20,-10", Scale="1,1,1") {
    rotationMode=BMM_Replace
    rotationSpace=BCS_BoneSpace
    scaleMode=BMM_Replace
    scaleSpace=BCS_BoneSpace
    translationMode=BMM_Replace
    translationSpace=BCS_BoneSpace
    alphaBoolBlend {blendOption=Linear}
    boneToModify {boneName=Dyn_Pivot_02}
    rotation {roll=-10, yaw=-20}
   }
   n6 = "Transform (Modify) Bone - Bone: Dyn_Pivot_01"(Alpha=1, Translation="0,0,0", Rotation="0,-10,0", Scale="1,1,1") {
    rotationMode=BMM_Replace
    rotationSpace=BCS_BoneSpace
    scaleMode=BMM_Replace
    scaleSpace=BCS_BoneSpace
    translationMode=BMM_Replace
    translationSpace=BCS_BoneSpace
    alphaBoolBlend {blendOption=Linear}
    boneToModify {boneName=Dyn_Pivot_01}
    rotation {yaw=-10}
   }
   AnimGraphNode_SubInput_0 = Input Pose()
   n7 = "Transform (Modify) Bone - Bone: Dyn_Pivot_01"(ComponentPose=Local To Component(LocalPose=@AnimGraphNode_SubInput_0), Alpha=1, Translation="0,0,0", Rotation="0,15,0", Scale="1,1,1") {
    rotationMode=BMM_Replace
    rotationSpace=BCS_BoneSpace
    scaleMode=BMM_Replace
    scaleSpace=BCS_BoneSpace
    translationMode=BMM_Replace
    translationSpace=BCS_BoneSpace
    alphaBoolBlend {blendOption=Linear}
    boneToModify {boneName=Dyn_Pivot_01}
    rotation {yaw=15}
   }
   n8 = "Transform (Modify) Bone - Bone: Dyn_Pivot_02"(Alpha=1, Translation="0,0,0", Rotation="0,-20,-35", Scale="1,1,1") {
    rotationMode=BMM_Replace
    rotationSpace=BCS_BoneSpace
    scaleMode=BMM_Replace
    scaleSpace=BCS_BoneSpace
    translationMode=BMM_Replace
    translationSpace=BCS_BoneSpace
    alphaBoolBlend {blendOption=Linear}
    boneToModify {boneName=Dyn_Pivot_02}
    rotation {roll=-35, yaw=-20}
   }
   n9 = Two Way Blend(A=Component To Local(ComponentPose=@n7), B=Component To Local(ComponentPose=@n8), Alpha=Get Stage4() -> (Stage4:double)) {blendNode {alphaBoolBlend {blendOption=Linear}, alphaScaleBiasClamp {bInterpResult=true, interpSpeedDecreasing=3, interpSpeedIncreasing=3}}}
   n10 = Two Way Blend(A=@n9, B=Component To Local(ComponentPose=@n6), Alpha=Get Stage1() -> (Stage1:double)) {blendNode {alphaBoolBlend {blendOption=Linear}, alphaScaleBiasClamp {bInterpResult=true, interpSpeedDecreasing=3, interpSpeedIncreasing=3}}}
   n11 = Two Way Blend(A=@n10, B=Component To Local(ComponentPose=@n5), Alpha=Get Stage5() -> (Stage5:double)) {blendNode {alphaBoolBlend {blendOption=Linear}, alphaScaleBiasClamp {bInterpResult=true, interpSpeedDecreasing=5, interpSpeedIncreasing=5}}}
   n12 = Two Way Blend(A=@n11, B=Component To Local(ComponentPose=@n4), Alpha=Get Stage2() -> (Stage2:double)) {blendNode {alphaBoolBlend {blendOption=Linear}, alphaScaleBiasClamp {bInterpResult=true, interpSpeedDecreasing=3, interpSpeedIncreasing=3}}}
   n13 = Two Way Blend(A=@n12, B=Component To Local(ComponentPose=@n3), Alpha=Get Stage6() -> (Stage6:double)) {blendNode {alphaBoolBlend {blendOption=Linear}, alphaScaleBiasClamp {bInterpResult=true, interpSpeedDecreasing=5, interpSpeedIncreasing=5}}}
   n14 = Two Way Blend(A=@n13, B=Component To Local(ComponentPose=@n2), Alpha=Get Stage3() -> (Stage3:double)) {blendNode {alphaBoolBlend {blendOption=Linear}, alphaScaleBiasClamp {bInterpResult=true, interpSpeedDecreasing=3, interpSpeedIncreasing=3}}}
   n15 = Two Way Blend(A=@n14, B=Component To Local(ComponentPose=@n1), Alpha=Get Stage7() -> (Stage7:double)) {blendNode {alphaBoolBlend {blendOption=Linear}, alphaScaleBiasClamp {bInterpResult=true, interpSpeedDecreasing=5, interpSpeedIncreasing=5}}}
   n16 = "Transform (Modify) Bone - Bone: Dyn_Satalite_Dish"(
    ComponentPose=Local To Component(LocalPose=@n15)
    Alpha=1
    Translation="0,0,0"
    Rotation=Break Anim Spinner!(AnimSpinner:"Anim Spinner Structure (by ref)"=Get Animation Spinner() -> (Animation Spinner:Anim Spinner Structure)) -> (
     Rotation:Rotator=~
     RotationAxis:Vector=~
     BaseRotationSpeed:float=~
     RotationRate:float=~
     InterpolationRateSpeedUp:float=~
     InterpolationRateSlowDown:float=~
     RotationAngle:float=~
    ) {is=title K2Node_BreakStruct}
    Scale="1,1,1"
   ) {rotationMode=BMM_Additive, rotationSpace=BCS_BoneSpace, alphaBoolBlend {blendOption=Linear}, boneToModify {boneName=Dyn_Satalite_Dish}}
   n17 = "Transform (Modify) Bone - Bone: Dyn_Master_Satalite_Dish"(ComponentPose=@n16, Alpha=1, Translation="0,0,0", Rotation=Get Dish_UpDown() -> (Dish_UpDown:Rotator), Scale="1,1,1") {
    rotationMode=BMM_Additive
    translationMode=BMM_Additive
    translationSpace=BCS_BoneSpace
    alphaBoolBlend {blendOption=Linear}
    boneToModify {boneName=Dyn_Master_Satalite_Dish}
   }
   n18 = RigidBody(ComponentPose=@n17) {
    bClampLinearTranslationLimitToRefPose=true
    bFreezeIncomingPoseOnStart=true
    bOverrideWorldGravity=true
    overridePhysicsAsset=&/Eternal_Vanguard_Female_Physics
    simulationSpace=BaseBoneSpace
    alphaBoolBlend {blendOption=Linear}
    baseBoneRef {boneName=b_AR_Root}
    componentAppliedLinearAccClamp=<5000 5000 5000>
    componentLinearAccScale=<0.3 0.2 0.2>
    componentLinearVelScale=<0.3 0.2 0.2>
    overrideWorldGravity {z=-980}
   }
   Output Pose(Result=Component To Local(ComponentPose=@n18)) {layerGroup=None}
  ]}
 EventGraph {nodes [
   n1 = Event BlueprintUpdateAnimation()
   > n2 = Sequence!(execute) -> (then_0, then_1, then_2:Exec)
   > UpdateAnimSpinner!(
    execute
    SpinInformation:"Anim Spinner Structure (by ref)"=Get Animation Spinner()
    DeltaSeconds:float=@n1.DeltaTimeX
    NewRotationRate:float=1
   ) -> (then)
   > Set Satelite_UD(Satelite_UD:Rotator=RInterpTo!(
     Current:Rotator=Get Satelite_UD() -> (Satelite_UD:Rotator)
     Target:Rotator=Get SatInterpTo() -> (SatInterpTo:Rotator)
     DeltaTime:float=@n1.DeltaTimeX
     InterpSpeed:float=1
    ) -> (ReturnValue:Rotator)) -> (Output_Get:Rotator)
   Do Once(execute=@n2.then_1, Reset=@n3.then)
   > Delay(Duration=2)
   > Set Stage4(Stage4:double=1) -> (Output_Get:double)
   > Delay(Duration=4)
   > Set Stage1(Stage1:double=1) -> (Output_Get:double)
   > Delay(Duration=5)
   > Set Stage5(Stage5:double=1) -> (Output_Get:double)
   > Delay(Duration=4)
   > Set Stage2(Stage2:double=1) -> (Output_Get:double)
   > Delay(Duration=5)
   > Set Stage6(Stage6:double=1) -> (Output_Get:double)
   > Delay(Duration=4)
   > Set Stage3(Stage3:double=1) -> (Output_Get:double)
   > Delay(Duration=5)
   > Set Stage7(Stage7:double=1) -> (Output_Get:double)
   > Delay(Duration=0.2)
   > Set Stage4(Stage4=0)
   > Delay(Duration=0.5)
   > Set Stage1(Stage1=0)
   > Delay(Duration=0.5)
   > Set Stage5(Stage5=0)
   > Delay(Duration=0.5)
   > Set Stage2(Stage2=0)
   > Delay(Duration=0.5)
   > Set Stage3(Stage3=0)
   > Delay(Duration=0.5)
   > Set Stage6(Stage6=0)
   > Delay(Duration=0.5)
   > Set Stage7(Stage7=0)
   > n3 = Delay(Duration=0.5)
   n4 = "Sin (Degrees)"!(A:double=float * float(A=GetGameTimeInSeconds(), B=100)) -> (ReturnValue:double)
   Set DishSway(execute=@n2.then_2, DishSway:Vector=MakeVector(Y=0, Z=float * float(A=@n4, B=2))) -> (Output_Get:Vector)
   n5 = "Sin (Degrees)"(A=float * float(A=GetGameTimeInSeconds(), B=50))
   > Set Dish_UpDown(Dish_UpDown:Rotator=MakeRotator(Roll=float * float(A=@n5, B=30))) -> (Output_Get:Rotator)
   TryGetPawnOwner()
   MakeRotator(Yaw=RandomFloatInRange(Min=0, Max=30))
   MakeRotator(Yaw=RandomFloatInRange(Min=0, Max=1))
   Event BlueprintInitializeAnimation()
  ]}
 NewEventGraph_0 {links [], nodes []}
]
variables [
 Animation Spinner {
  type=AnimSpinner
  default {baseRotationSpeed=30, interpolationRateSlowDown=1, interpolationRateSpeedUp=5, rotationAngle=0, rotationRate=1, rotationAxis=<20 0 0>}
 }
 Dish_Point {type=Vector}
 Dish_UpDown {type=Rotator}
 DishSway {type=Vector}
 Original_Pose {type=number}
 Satelite_UD {type=Rotator}
 satInterpFrom {type=Rotator}
 SatInterpTo {type=Rotator}
 Stage1 {type=number}
 Stage2 {type=number}
 Stage3 {type=number}
 Stage4 {type=number}
 Stage5 {type=number}
 Stage6 {type=number}
 Stage7 {type=number}
]
